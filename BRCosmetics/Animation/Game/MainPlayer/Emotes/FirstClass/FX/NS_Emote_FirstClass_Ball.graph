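# Niagara system for the FirstClass emote: a glowing orange fire ball that follows a skeletal mesh, with a short trail.
# Fresnel_ball bursts a mesh with a point light, FireBits_local adds animated flipbook fire sprites, and FN_AttachedRibbon draws the trail through two ribbon renderers.
# Ball and sprites resample the skeletal mesh location every update to stay attached.
# User.Size scales all three emitters, while User.Intensity boosts fire spawn rate and ball and fire brightness.
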NiagaraSystem /BRCosmetics/Animation/Game/MainPlayer/Emotes/FirstClass/FX/NS_Emote_FirstClass_Ball
dependencies {
 usedDataInterfaces [&NiagaraDataInterfaceCurve, &NiagaraDataInterfaceColorCurve]
 usedDynamicInputs []
 usedModules [
  &self:SystemSpawnScript.SystemScriptSource.SystemScriptGraph.NiagaraNodeAssignment_0.SetVariables_#1
  &SystemState
 ]
}
emitters {
 FN_AttachedRibbon {
  emitterUpdateScript [
   &EmitterState()
   &FN_SpawnRibbon_v2(Spawn Mode=::NewEnumerator0, SpawnRate=30)
   &FN_Integrator() {enabled=false}
  ]
  particleSpawnScript [
   /Game/Effects/Niagara/Modules/Ribbons/FN_InitializeRibbon_v4(
    Enable Skeletal Position=true
    Lifetime=0.15
    Ribbon Width:NiagaraFloat=&Multiply_Float(A=33, B=$User.Size)
    SkeletalMesh:NiagaraDataInterfaceSkeletalMesh=$System.Skeletal Mesh
   )
   &FN_FlipRibbonV()
   &AddVelocity(Velocity=<0 0 200>)
  ]
  particleUpdateScript [
   &SolveForcesAndVelocity(Clamp Velocity=false, Limit Acceleration=false)
   /Game/Effects/Niagara/Modules/Ribbons/FN_UpdateRibbon_v4(
    Enable Skeletal Position=true
    Initial Color:LinearColor=&MultiplyLinearColor(Linear Color=$Particles.Color, Scale Factor=&MakeVector4(X=1, Y=1, Z=1, W=$User.Size))
    SkeletalMesh:NiagaraDataInterfaceSkeletalMesh=$System.Skeletal Mesh
   )
   &FN_RibbonMaterial() {enabled=false}
   &ScaleColor(ScaleA=true, ScaleRGB=true) {enabled=false}
  ]
  rendererClasses [&NiagaraRibbonRendererProperties]
  renderers [
   {rendererClass=&NiagaraRibbonRendererProperties}
   {rendererClass=&NiagaraRibbonRendererProperties, rendererIndex=1}
  ]
 }
 FireBits_local {
  emitterUpdateScript [
   &EmitterState()
   &SpawnRate(
    SpawnRate=&Add_Float(A=15, B=&Multiply_Float(A=$User.Intensity, B=10))
    Use Spawn Probability=false
   )
  ]
  particleSpawnScript [
   &InitializeParticle(
    Lifetime=0.5
    Sprite Rotation Mode=::NewEnumerator2
    Sprite Size Mode=::NewEnumerator1
    Uniform Sprite Size Max=55
    Uniform Sprite Size Min=40
    UsePositionOffset=false
    Write Material Random=true
   )
   &DynamicMaterialParameters(
    Index 0 Param 2=&UniformRangedFloat(Minimum=0, Maximum=1)
    Index 0 Param 4=&UniformRangedFloat(Minimum=3, Maximum=5)
    Module.Param0WriteEnabled=false
    Module.Param1WriteEnabled=true
    Module.Param2WriteEnabled=false
    Module.Param3WriteEnabled=true
   )
  ]
  particleUpdateScript [
   &ParticleState()
   &ShapeLocation(MaskSpawnGroup=false, Override Seed=false, Sphere Radius=15, Sphere Surface Distribution=1, UseOwnerScale=true)
   &SkeletalMeshLocation(
    Bone Sampling Mode=::NewEnumerator4
    MaskSpawnGroup=false
    Sampled Position Offset=$Particles.ShapeLocation.ShapePosition
    Skeletal Mesh=$System.Skeletal Mesh
   )
   &ScaleSpriteSize(Scale Factor=&Vector2DFromFloat(Value=&FloatFromCurve(
      FloatCurve=DataInterface() {propertyValues {Curve {keys [
          {arriveTangent=2.4034, interpMode=RCIM_Cubic, leaveTangent=2.4033, tangentMode=RCTM_User, time=0.0002, value=0.15}
          {arriveTangent=0.5342, interpMode=RCIM_Cubic, leaveTangent=0.5342, tangentMode=RCTM_User, time=0.3, value=0.575}
          {arriveTangent=-0.1019, interpMode=RCIM_Cubic, leaveTangent=-0.1019, tangentMode=RCTM_User, time=1.0041, value=0.7174}
         ]}}}
      Scale Curve=$User.Size
     )))
   &SpriteRotationRate(Rotation Rate=50)
   &SolveForcesAndVelocity(Clamp Velocity=false, Limit Acceleration=false)
   /SetVariables(Particles.SubImageIndex=&FloatFromCurve(
     FloatCurve=DataInterface() {propertyValues {Curve {keys [{tangentMode=RCTM_User, time=0, value=0}, {tangentMode=RCTM_User, time=1, value=1}]}}}
     Scale Curve=40
    )) {isSetParameters=true}
   &DynamicMaterialParameters(
    Index 0 Param 1=&FloatFromCurve(FloatCurve=DataInterface() {propertyValues {Curve {keys [<at 0 0 RCIM_Cubic>, {arriveTangent=0.9989, interpMode=RCIM_Cubic, leaveTangent=0.9989, time=0.6, value=0.25}, <at 1.0011 1 RCIM_Cubic>]}}})
    Index 0 Param 2=2.5
    Index 0 Param 3=1
    Index 0 Param 4=3
    Module.Param0WriteEnabled=true
    Module.Param1WriteEnabled=false
    Module.Param2WriteEnabled=true
    Module.Param3WriteEnabled=false
   ) {module=DynamicMaterialParameters001}
   &ScaleColor(
    Scale RGB=&Multiply_VectorByFloat(
     Vector=&MakeVectorFromLinearColorRGB(Color=&ColorFromCurve(ColorCurve=DataInterface() {propertyValues {
         AlphaCurve {keys [<at 0 1 RCIM_Cubic>]}
         BlueCurve {keys [<at 0 0 RCIM_Cubic>,<at 0.1964 0>,<at 0.8699 0 RCIM_Cubic>]}
         GreenCurve {keys [<at 0 0.939 RCIM_Cubic>,<at 0.1964 0.2885>,<at 0.8699 0.0542 RCIM_Cubic>]}
         RedCurve {keys [<at 0 3 RCIM_Cubic>,<at 0.1964 2>,<at 0.8699 1 RCIM_Cubic>]}
        }}))
     Float=&Add_Float(A=1, B=&Multiply_Float(A=$User.Intensity, B=2.5))
    )
    ScaleA=true
    ScaleRGB=true
   )
   &CameraOffset(Camera Offset Amount=10)
  ]
  rendererClasses [&NiagaraSpriteRendererProperties]
  renderers [{rendererClass=&NiagaraSpriteRendererProperties}]
 }
 Fresnel_ball {
  emitterUpdateScript [
   &EmitterState()
   &SpawnBurst_Instantaneous(Use Spawn Probability=false, UseLoopCountLimit=false)
  ]
  particleSpawnScript [&InitializeParticle(
    Color=<rgba 2 0.2834 0 1>
    Lifetime=6
    Mesh Scale Mode=::NewEnumerator3
    Mesh Uniform Scale=0.1
    Sprite Size Mode=::NewEnumerator3
    Uniform Sprite Size=27
    UsePositionOffset=false
    Write Material Random=true
   )]
  particleUpdateScript [
   &ParticleState()
   &SkeletalMeshLocation(Bone Sampling Mode=::NewEnumerator4, MaskSpawnGroup=false, Skeletal Mesh=$System.Skeletal Mesh)
   &CameraOffset(Camera Offset Amount=5)
   &ScaleMeshSize(Scale Factor=&VectorFromFloat(Value=$User.Size))
   &FN_Light_Attributes(
    Color=$Particles.Color
    Create Color Attribute=true
    Create Exponent Attribute=true
    Create Radius Attribute=true
    Exponent=7
    Radius=&Multiply_Float(A=90, B=$User.Size)
    Scale Color=9
   )
   &ScaleColor(
    Scale RGB=&VectorFromFloat(Value=&Add_Float(A=1, B=&Multiply_Float(A=$User.Intensity, B=5)))
    ScaleA=true
    ScaleRGB=true
   )
  ]
  rendererClasses [&NiagaraMeshRendererProperties, &NiagaraLightRendererProperties]
  renderers [
   {rendererClass=&NiagaraMeshRendererProperties, rendererIndex=0}
   {rendererClass=&NiagaraLightRendererProperties, rendererIndex=1}
  ]
 }
}
summary {
 systemName=NS_Emote_FirstClass_Ball
 emitters [
  {emitterName=Fresnel_ball, rendererClasses [&NiagaraMeshRendererProperties, &NiagaraLightRendererProperties]}
  {emitterName=FN_AttachedRibbon, rendererClasses [&NiagaraRibbonRendererProperties]}
  {emitterName=FireBits_local, rendererClasses [&NiagaraSpriteRendererProperties]}
 ]
 userVariables [
  User.Intensity {type=NiagaraFloat, defaultValue=1, description=""}
  User.RibbonIDOffset {type=NiagaraInt32}
  User.Size {type=NiagaraFloat, defaultValue=1, description=""}
  User.SkeletalMesh {
   type=NiagaraDataInterfaceSkeletalMesh
   defaultValue {
    struct=NiagaraExt_VariableValue_DataInterface
    value {dataInterface=&self:NiagaraDataInterfaceSkeletalMesh_0}
   }
  }
 ]
}
