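# Niagara system for the FirstClass emote's ball impact, a short fiery burst with smoke.
# Five burst emitters layer sparks stretched by speed, animated fire sprites, an expanding splash mesh, smoke puffs and a fading glow with a light renderer.
# All particles share a world-space System.PosOffset at spawn, and curves drive color, scale and material parameters over life.
# The exposed ImpactColor and IsWeakPoint user variables are not referenced by the emitter modules shown.
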NiagaraSystem /BRCosmetics/Animation/Game/MainPlayer/Emotes/FirstClass/FX/NS_Emote_FirstClass_BallImpact
dependencies {
 usedDataInterfaces [
  &NiagaraDataInterfaceCurve
  &NiagaraDataInterfaceVectorCurve
  &NiagaraDataInterfaceColorCurve
  &NiagaraDataInterfaceMeshRendererInfo
 ]
 usedDynamicInputs [
  &RandomRangeVector
  &MakeCustomFloatFromBool
  &SimulationPosition
 ]
 usedModules [
  &self:SystemSpawnScript.SystemScriptSource.SystemScriptGraph.NiagaraNodeAssignment_0.SetVariables_#1
  &SystemState
 ]
}
emitters {
 Embers {
  emitterUpdateScript [
   &EmitterState()
   &SpawnBurst_Instantaneous(Spawn Count=21, Use Spawn Probability=false, UseLoopCountLimit=false)
  ]
  particleSpawnScript [
   &InitializeParticle(
    Color Mode=::NewEnumerator0
    Lifetime Max=0.7
    Lifetime Min=0.2
    Lifetime Mode=::NewEnumerator1
    Position Offset=$System.PosOffset
    Position Offset Coordinate Space=::World
    Sprite Size Mode=::NewEnumerator1
    Uniform Sprite Size Max=6
    Uniform Sprite Size Min=2
    UsePositionOffset=true
    Write Material Random=true
   )
   &ShapeLocation(
    MaskSpawnGroup=false
    Offset Coordinate Space=::Local
    Offset Mode=::NewEnumerator0
    Override Seed=false
    Sphere Radius=0
    UseOwnerScale=true
   )
   &AddVelocity(
    Cone Angle=180
    Cone Axis=<0 0 1>
    Module.UseConeFalloff=false
    Module.UseVelDistribution=false
    Override Seed=false
    Velocity Mode=::NewEnumerator2
    Velocity Speed=&RandomRangeFloat(Minimum=100, Maximum=700)
   )
   &CurlNoiseForce(Noise Strength=400, Pan Noise=false)
   &ApplyInitialForces()
  ]
  particleUpdateScript [
   &ParticleState()
   &Drag(Drag=6, Use Linear Drag=true, Use Rotational Drag=true)
   &GravityForce(Gravity=<0 0 -444>)
   &SolveForcesAndVelocity(Clamp Velocity=false, Limit Acceleration=false)
   &ScaleColor(
    Scale RGB=&MakeVectorFromLinearColorRGB(Color=&ColorFromCurve(ColorCurve=DataInterface() {propertyValues {
        AlphaCurve {keys [<at 0 1 RCIM_Cubic>]}
        BlueCurve {keys [<at 0 0 RCIM_Cubic>,<at 0.333 0>,<at 1 0 RCIM_Cubic>]}
        GreenCurve {keys [<at 0 3.5549 RCIM_Cubic>, <at 0.333 0.15>, {interpMode=RCIM_Cubic, leaveTangent=-0.4498, time=1, value=0}]}
        RedCurve {keys [<at 0 10 RCIM_Cubic>, <at 0.333 3>, {interpMode=RCIM_Cubic, leaveTangent=-8.9955, time=1, value=1}]}
       }}))
    ScaleA=true
    ScaleRGB=true
   )
   &ScaleSpriteSizeBySpeed(Max Scale Factor=<1 3>, Min Scale Factor=<1 -3>, Velocity Threshold=222)
  ]
  rendererClasses [&NiagaraSpriteRendererProperties]
  renderers [{rendererClass=&NiagaraSpriteRendererProperties}]
 }
 FN_FireflyJar_Exlosion_Fire {
  emitterUpdateScript [
   &EmitterState()
   &SpawnBurst_Instantaneous(Spawn Count=4, Use Spawn Probability=false)
  ]
  particleSpawnScript [
   &InitializeParticle(
    Lifetime=&UniformRangedFloat(Minimum=0.45, Maximum=0.6)
    Mass Mode=::NewEnumerator1
    Position Offset=$System.PosOffset
    Position Offset Coordinate Space=::World
    Sprite Rotation Mode=::NewEnumerator2
    Sprite Size Mode=::NewEnumerator1
    Uniform Sprite Size Max=100
    Uniform Sprite Size Min=50
    UsePositionOffset=true
    Write Material Random=true
   ) {module=InitializeParticle001}
   &DynamicMaterialParameters(
    Index 0 Param 2=&UniformRangedFloat(Minimum=0, Maximum=1)
    Index 0 Param 4=&UniformRangedFloat(Minimum=3, Maximum=5)
    Module.Param0WriteEnabled=false
    Module.Param1WriteEnabled=true
    Module.Param2WriteEnabled=false
    Module.Param3WriteEnabled=true
   )
   &ShapeLocation(
    Cylinder Height=0
    Cylinder Radius=15
    MaskSpawnGroup=false
    Override Seed=false
    Shape Primitive=::NewEnumerator1
    Surface Only=false
    UseOwnerScale=true
   )
   &AddVelocity(Velocity=$Particles.ShapeLocation.ShapeNormal, Velocity Speed Scale=90)
  ]
  particleUpdateScript [
   &ParticleState()
   &ScaleSpriteSize(Scale Factor=&Vector2DFromFloat(Value=&FloatFromCurve(FloatCurve=DataInterface() {propertyValues {Curve {keys [
          {arriveTangent=2.4033, interpMode=RCIM_Cubic, leaveTangent=2.4033, tangentMode=RCTM_User, time=0.0002, value=0.15}
          {arriveTangent=0.5342, interpMode=RCIM_Cubic, leaveTangent=0.5342, tangentMode=RCTM_User, time=0.3, value=0.575}
          {arriveTangent=-0.1019, interpMode=RCIM_Cubic, leaveTangent=-0.1019, tangentMode=RCTM_User, time=1.0041, value=0.7174}
         ]}}})))
   /SetVariables(Particles.SubImageIndex=&FloatFromCurve(
     FloatCurve=DataInterface() {propertyValues {Curve {keys [
         {arriveTangent=2.1444, interpMode=RCIM_Cubic, leaveTangent=2.1444, tangentMode=RCTM_User, time=0, value=0}
         {arriveTangent=0.1938, interpMode=RCIM_Cubic, leaveTangent=0.1938, tangentMode=RCTM_User, time=1, value=1}
        ]}}}
     Scale Curve=40
    )) {isSetParameters=true}
   &DynamicMaterialParameters(
    Index 0 Param 1=&FloatFromCurve(FloatCurve=DataInterface() {propertyValues {Curve {keys [<at 0 0>,<at 0.8 1>,<at 1.0011 1>]}}})
    Index 0 Param 2=2.5
    Index 0 Param 3=1
    Index 0 Param 4=3
    Module.Param0WriteEnabled=true
    Module.Param1WriteEnabled=false
    Module.Param2WriteEnabled=true
    Module.Param3WriteEnabled=false
   ) {module=DynamicMaterialParameters001}
   &GravityForce(Gravity=&MakeVector(
     X=0
     Y=0
     Z=&FloatFromCurve(FloatCurve=DataInterface() {propertyValues {Curve {keys [<at 0 0>,<at 0.6 0>,<at 0.7124 0.9779>,<at 1 1>]}}}, Scale Curve=100)
    ))
   &Drag(Drag=2.5, Use Linear Drag=true, Use Rotational Drag=false)
   &SolveForcesAndVelocity(Clamp Velocity=false, Limit Acceleration=false)
   &ScaleColor(
    Scale RGB=&MakeVectorFromLinearColorRGB(Color=&ColorFromCurve(ColorCurve=DataInterface() {propertyValues {
        AlphaCurve {keys [<at 0 1 RCIM_Cubic>]}
        BlueCurve {keys [<at 0 0 RCIM_Cubic>,<at 0.1964 0>,<at 0.8699 0 RCIM_Cubic>]}
        GreenCurve {keys [<at 0 0.939 RCIM_Cubic>,<at 0.1964 0.2885>,<at 0.8699 0.0542 RCIM_Cubic>]}
        RedCurve {keys [<at 0 3 RCIM_Cubic>,<at 0.1964 2>,<at 0.8699 1 RCIM_Cubic>]}
       }}))
    ScaleA=true
    ScaleRGB=true
   )
   &SpriteRotationRate(Rotation Rate=50)
  ]
  rendererClasses [&NiagaraSpriteRendererProperties]
  renderers [{rendererClass=&NiagaraSpriteRendererProperties}]
 }
 FN_Light {
  emitterUpdateScript [
   &EmitterState()
   &SpawnBurst_Instantaneous(Spawn Count=1, Use Spawn Probability=false)
  ]
  particleSpawnScript [&InitializeParticle(
    Lifetime=0.35
    Position Offset=$System.PosOffset
    Position Offset Coordinate Space=::World
    Sprite Rotation Angle=&UniformRangedFloat(Minimum=0, Maximum=360)
    Sprite Rotation Mode=::NewEnumerator0
    Sprite Size Mode=::NewEnumerator3
    Uniform Sprite Size=300
    UsePositionOffset=true
    Write Material Random=true
   ) {module=InitializeParticle001}]
  particleUpdateScript [
   &ParticleState()
   &ScaleColor(
    Scale Alpha=&FloatFromCurve(FloatCurve=DataInterface() {propertyValues {Curve {keys [<at 0 0.5 RCIM_Cubic>,<at 1 0>]}}})
    Scale RGB=&MakeVectorFromLinearColorRGB(Color=&ColorFromCurve(ColorCurve=DataInterface() {propertyValues {
        AlphaCurve {keys [<at 0 1 RCIM_Cubic>]}
        BlueCurve {keys [{arriveTangent=0.3517, interpMode=RCIM_Cubic, time=0.0193, value=0.224}, <at 0.1705 0.0503>, <at 0.4647 0>]}
        GreenCurve {keys [{arriveTangent=0.8075, interpMode=RCIM_Cubic, time=0.0193, value=0.7829}, <at 0.1705 0.4038>, <at 0.4647 0.1069>]}
        RedCurve {keys [{arriveTangent=3.7024, interpMode=RCIM_Cubic, time=0.0193, value=1}, <at 0.1705 1>, <at 0.4647 1>]}
       }}))
    ScaleA=true
    ScaleRGB=true
   )
   &SolveForcesAndVelocity(Clamp Velocity=false, Limit Acceleration=false)
   &ScaleSpriteSize(Scale Factor=&Vector2DFromFloat(Value=&FloatFromCurve(FloatCurve=DataInterface() {propertyValues {Curve {keys [<at 0 0.1 RCIM_Cubic>,<at 1 1>]}}}, Scale Curve=1.5)))
   &FN_Light_Attributes(
    Color=$Particles.Color
    Create Color Attribute=true
    Create Exponent Attribute=true
    Create Position Attribute=true
    Create Radius Attribute=true
    Exponent=6
    Position=<0 0 5>
    Radius=35
    Scale Color=&FloatFromCurve(
     FloatCurve=DataInterface() {propertyValues {Curve {keys [
         {interpMode=RCIM_Cubic, tangentMode=RCTM_User, time=0, value=1}
         {arriveTangent=-1, interpMode=RCIM_Cubic, leaveTangent=-1, time=0.6, value=0.6}
         {interpMode=RCIM_Cubic, tangentMode=RCTM_User, time=1, value=0}
        ]}}}
     Scale Curve=15
    )
   )
  ]
  rendererClasses [&NiagaraLightRendererProperties, &NiagaraSpriteRendererProperties]
  renderers [
   {rendererClass=&NiagaraLightRendererProperties, rendererIndex=0}
   {rendererClass=&NiagaraSpriteRendererProperties, rendererIndex=1}
  ]
 }
 LavaSplash001 {
  emitterUpdateScript [
   &EmitterState()
   &SpawnBurst_Instantaneous(Use Spawn Probability=false, UseLoopCountLimit=false)
  ]
  particleSpawnScript [
   &InitializeParticle(
    Lifetime=0.35
    Mesh Scale Mode=::NewEnumerator3
    Mesh Uniform Scale=0.5
    Position Offset=$System.PosOffset
    Position Offset Coordinate Space=::World
    Sprite Rotation Mode=::NewEnumerator2
    Sprite Size Mode=::NewEnumerator3
    Uniform Sprite Size=150
    UsePositionOffset=true
    Write Material Random=true
   )
   &InitialMeshOrientation(Mesh Axis To Orient=<0 0 1>, Mesh Orientation Mode=::NewEnumerator2, Use Rotation Vector=false)
  ]
  particleUpdateScript [
   &ParticleState()
   &ScaleMeshSize(Scale Factor=&VectorFromCurve(VectorCurve=DataInterface() {propertyValues {
       XCurve {keys [
         {arriveTangent=2.7025, interpMode=RCIM_Cubic, leaveTangent=2.7025, tangentMode=RCTM_User, time=0, value=0}
         {arriveTangent=0.6, interpMode=RCIM_Cubic, leaveTangent=0.6, time=0.3, value=0.6}
         <at 1 0.6 RCIM_Cubic>
        ]}
       YCurve {keys [
         {arriveTangent=2.3749, interpMode=RCIM_Cubic, leaveTangent=2.3749, tangentMode=RCTM_User, time=0, value=0}
         {arriveTangent=1, interpMode=RCIM_Cubic, leaveTangent=1, time=0.3, value=0.825}
         <at 1 1 RCIM_Cubic>
        ]}
       ZCurve {keys [
         {arriveTangent=2.3749, interpMode=RCIM_Cubic, leaveTangent=2.3749, tangentMode=RCTM_User, time=0, value=0}
         {arriveTangent=1, interpMode=RCIM_Cubic, leaveTangent=1, time=0.3, value=0.825}
         <at 1 1 RCIM_Cubic>
        ]}
      }}))
   &ScaleSpriteSize(Uniform Curve Sprite Scale=DataInterface() {propertyValues {Curve {keys [
        {arriveTangent=1.0404, interpMode=RCIM_Cubic, leaveTangent=1.0404, tangentMode=RCTM_Break, time=0, value=0.25}
        {arriveTangent=-0.0405, interpMode=RCIM_Cubic, leaveTangent=-0.0405, tangentMode=RCTM_User, time=1, value=1}
       ]}}})
   &ScaleColor(
    Scale Alpha=&FloatFromCurve(FloatCurve=DataInterface() {propertyValues {Curve {keys [
         {interpMode=RCIM_Cubic, leaveTangent=-0.8793, tangentMode=RCTM_Break, time=0, value=1}
         {arriveTangent=-1, interpMode=RCIM_Cubic, leaveTangent=-1, time=0.7, value=0.7}
         {arriveTangent=0.0015, interpMode=RCIM_Cubic, leaveTangent=-1.7296, tangentMode=RCTM_Break, time=1, value=0}
        ]}}})
    ScaleA=true
    ScaleRGB=true
   )
   &SolveForcesAndVelocity(Clamp Velocity=false, Limit Acceleration=false)
   &DynamicMaterialParameters(
    Index 0 Param 1=&FloatFromCurve(FloatCurve=DataInterface() {propertyValues {Curve {keys [{interpMode=RCIM_Cubic, tangentMode=RCTM_User, time=0, value=0}, {interpMode=RCIM_Cubic, tangentMode=RCTM_User, time=1, value=1}]}}})
    Module.Index0Param0Write=true
    Module.Index0Param1Write=true
    Module.Index0Param2Write=true
    Module.Index0Param3Write=true
   )
  ]
  rendererClasses [&NiagaraMeshRendererProperties]
  renderers [{rendererClass=&NiagaraMeshRendererProperties, rendererIndex=0}]
 }
 SmokePuff_Gen5 {
  emitterUpdateScript [
   &EmitterState()
   &SpawnBurst_Instantaneous(Spawn Count=&RandomRangeInt(Minimum=2, Maximum=3), Use Spawn Probability=false)
  ]
  particleSpawnScript [
   /Niagara/Modules/Spawn/Initialization/InitializeParticle(
    Color=&RandomRangeLinearColor(Minimum=<rgba 1 1 1 0.15>, Maximum=<rgba 1 1 1 0.2>)
    Lifetime=&RandomRangeFloat(Minimum=0.3, Maximum=0.4, Evaluation Type=::NewEnumerator1)
    Module.Write Color=true
    Module.Write Lifetime=true
    Module.Write Mass=true
    Module.Write Material Random=true
    Module.Write Position=false
    Module.Write Scale=false
    Module.Write SpriteRotation=true
    Module.Write SpriteSize=true
    Sprite Rotation=&RandomRangeFloat(Minimum=-15, Maximum=15)
    Sprite Size=&Vector2DFromFloat(Value=&RandomRangeFloat(Minimum=70, Maximum=90, Evaluation Type=::NewEnumerator1))
   ) {enabled=false}
   /Niagara/Modules/Spawn/Initialization/V2/InitializeParticle(
    Color=&RandomRangeLinearColor(Minimum=<rgba 1 1 1 0.15>, Maximum=<rgba 1 1 1 0.2>)
    Lifetime=&RandomRangeFloat(Minimum=0.5, Maximum=1)
    Mass Mode=::NewEnumerator1
    Position Offset=$System.PosOffset
    Position Offset Coordinate Space=::World
    Sprite Rotation Angle=&RandomRangeFloat(Minimum=-15, Maximum=15)
    Sprite Rotation Mode=::NewEnumerator0
    Sprite Size=&Vector2DFromFloat(Value=&RandomRangeFloat(Minimum=100, Maximum=120))
    Sprite Size Mode=::NewEnumerator4
    UsePositionOffset=true
    Write Material Random=true
   ) {module=InitializeParticle001}
   &CameraOffset(Camera Offset Amount=&RandomRangeFloat(Minimum=40, Maximum=50, Evaluation Type=::NewEnumerator1)) {enabled=false}
   /Niagara/Modules/Spawn/Location/SphereLocation(MaskSpawnGroup=false, Module.Override Randomness=false, Sphere Radius=1, Surface Only=false)
   &AddVelocity(Velocity=&Multiply_VectorByFloat(Vector=$Output.SphereLocation.SphereVector, Float=200))
  ]
  particleUpdateScript [
   &ParticleState()
   &ScaleSpriteSize(Scale Factor=&Vector2DFromFloat(Value=&FloatFromCurve(
      FloatCurve=DataInterface() {propertyValues {Curve {keys [
          {arriveTangent=1.6027, interpMode=RCIM_Cubic, leaveTangent=1.6027, tangentMode=RCTM_User, time=-0.0011, value=0.0573}
          {interpMode=RCIM_Cubic, leaveTangent=0.0935, time=1, value=1}
         ]}}}
      Scale Curve=0.8
     )))
   &ScaleColor(
    Scale Alpha=&FloatFromCurve(FloatCurve=DataInterface() {propertyValues {Curve {keys [<at 0 0>,<at 0.25 0.5>,<at 1 0>]}}})
    Scale RGB=&VectorFromCurve(VectorCurve=DataInterface() {propertyValues {XCurve {keys [<at 0 1>,<at 1 1>]}, YCurve {keys [<at 0 0>,<at 1 1>]}, ZCurve {keys [<at 0 0>,<at 1 1>]}}})
    ScaleA=true
    ScaleRGB=true
   )
   &DynamicMaterialParameters(
    Index 0 Param 1=&FloatFromCurve(
     FloatCurve=DataInterface() {propertyValues {Curve {keys [
         <at 0 0 RCIM_Cubic>
         {arriveTangent=1.5947, interpMode=RCIM_Cubic, leaveTangent=1.5947, tangentMode=RCTM_User, time=0.2557, value=0.5257}
         <at 1 1 RCIM_Cubic>
        ]}}}
     CurveIndex=$Particles.Age
     Scale Curve=0.25
    )
    Module.Param0WriteEnabled=true
    Module.Param1WriteEnabled=true
    Module.Param2WriteEnabled=true
    Module.Param3WriteEnabled=true
   ) {module=DynamicMaterialParameters001}
   &Drag(Drag=&RandomRangeFloat(Minimum=3, Maximum=4), Use Linear Drag=true, Use Rotational Drag=true)
   &SolveForcesAndVelocity(Clamp Velocity=false, Limit Acceleration=false)
  ]
  rendererClasses [&NiagaraSpriteRendererProperties]
  renderers [
   {rendererClass=&NiagaraSpriteRendererProperties}
   {rendererClass=&NiagaraSpriteRendererProperties, rendererIndex=1}
  ]
 }
}
summary {
 systemName=NS_Emote_FirstClass_BallImpact
 emitters [
  {emitterName=SmokePuff_Gen5, rendererClasses [&NiagaraSpriteRendererProperties]}
  {emitterName=Embers, rendererClasses [&NiagaraSpriteRendererProperties]}
  {emitterName=LavaSplash001, rendererClasses [&NiagaraMeshRendererProperties]}
  {emitterName=FN_FireflyJar_Exlosion_Fire, rendererClasses [&NiagaraSpriteRendererProperties]}
  {emitterName=FN_Light, rendererClasses [&NiagaraLightRendererProperties, &NiagaraSpriteRendererProperties]}
 ]
 userVariables [
  User.ImpactColor {type=LinearColor, description="", defaultValue {struct=LinearColor, value=<rgba 1 1 1 1>}}
  User.IsWeakPoint {type=NiagaraBool}
 ]
}
