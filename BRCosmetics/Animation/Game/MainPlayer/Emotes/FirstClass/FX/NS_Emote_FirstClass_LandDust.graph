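# Niagara dust puff for the landing in the FirstClass main-player emote.
# One Dust emitter bursts short-lived sprites outward, slowed by drag and lifted by upward acceleration, with camera offset and two sprite renderers.
# Sprites grow and fade over life while curves drive dynamic material parameters.
# A system-level SetFootDustFXColorArrayIndex module accompanies user colour-array and surface-type inputs, though the emitter initialises particles to fixed translucent grey.
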NiagaraSystem /BRCosmetics/Animation/Game/MainPlayer/Emotes/FirstClass/FX/NS_Emote_FirstClass_LandDust
dependencies {
 usedDataInterfaces [&NiagaraDataInterfaceCurve, &NiagaraDataInterfaceColorCurve]
 usedDynamicInputs [&RandomRangeVector]
 usedModules [
  &SystemState
  &self:SetFootDustFXColorArrayIndex
 ]
}
emitters {Dust {
  emitterUpdateScript [
   &EmitterState()
   &SpawnBurst_Instantaneous(
    Spawn Count=&RandomRangeInt(Minimum=10, Maximum=12)
    Spawn Time=0.05
    Use Spawn Probability=true
    UseLoopCountLimit=false
   )
  ]
  particleSpawnScript [
   &InitializeParticle(
    Color=<rgba 0.3281 0.3281 0.3281 0.5>
    Lifetime Max=0.4
    Lifetime Min=0.27
    Lifetime Mode=::NewEnumerator1
    Sprite Rotation Mode=::NewEnumerator2
    Sprite Size Mode=::NewEnumerator1
    Sprite UV Mode=::NewEnumerator3
    Uniform Sprite Size Max=40
    Uniform Sprite Size Min=30
    UsePositionOffset=false
    Write Material Random=true
   )
   &ShapeLocation(
    Handle Radius=1
    Large Radius=5
    MaskSpawnGroup=false
    Override Seed=false
    Rotation Mode=::NewEnumerator2
    Shape Primitive=::NewEnumerator3
    UseOwnerScale=true
    Yaw / Pitch / Roll=<0 0 1>
   )
   &AddVelocity(Module.Use Falloff Distance=false, Override Seed=false, Velocity Mode=::NewEnumerator1, Velocity Speed=500) {module=AddVelocity001}
   &CameraOffset(Camera Offset Amount=28)
  ]
  particleUpdateScript [
   &Drag(Drag=4, Use Linear Drag=true, Use Rotational Drag=false)
   &AccelerationForce(Acceleration=&MakeVector(X=0, Y=0, Z=&RandomRangeFloat(Minimum=200, Maximum=300)))
   &SolveForcesAndVelocity(Clamp Velocity=false, Limit Acceleration=false)
   &ParticleState()
   &SpriteRotationRate(Rotation Rate=&RandomRangeFloat(Minimum=-5, Maximum=5))
   &DynamicMaterialParameters(
    Index 0 Param 1=&FloatFromCurve(FloatCurve=DataInterface() {propertyValues {Curve {keys [{arriveTangent=1.2241, interpMode=RCIM_Cubic, leaveTangent=1.2241, tangentMode=RCTM_User, time=0, value=0.5}, <at 1 1>]}}})
    Index 0 Param 2=1
    Index 0 Param 4=1
    Index 2 Param 1=&FloatFromCurve(FloatCurve=DataInterface() {propertyValues {Curve {keys [
         {arriveTangent=3.7909, interpMode=RCIM_Cubic, leaveTangent=3.7909, tangentMode=RCTM_User, time=0, value=0}
         {arriveTangent=1.8782, interpMode=RCIM_Cubic, leaveTangent=1.8782, tangentMode=RCTM_User, time=0.2, value=0.55}
         {arriveTangent=0.0797, interpMode=RCIM_Cubic, leaveTangent=0.0797, tangentMode=RCTM_User, time=1, value=1}
        ]}}})
    Index 2 Param 2=0.4
    Index 2 Param 3=&FloatFromCurve(FloatCurve=DataInterface() {propertyValues {Curve {keys [
         {arriveTangent=3.7909, interpMode=RCIM_Cubic, leaveTangent=3.7909, tangentMode=RCTM_User, time=0, value=0}
         {arriveTangent=1.8782, interpMode=RCIM_Cubic, leaveTangent=1.8782, tangentMode=RCTM_User, time=0.2, value=0.55}
         {arriveTangent=0.0797, interpMode=RCIM_Cubic, leaveTangent=0.0797, tangentMode=RCTM_User, time=1, value=1}
        ]}}})
    Index 2 Param 4=&FloatFromCurve(FloatCurve=DataInterface() {propertyValues {Curve {keys [
         {arriveTangent=3.7909, interpMode=RCIM_Cubic, leaveTangent=3.7909, tangentMode=RCTM_User, time=0, value=0}
         {arriveTangent=1.8782, interpMode=RCIM_Cubic, leaveTangent=1.8782, tangentMode=RCTM_User, time=0.2, value=0.55}
         {arriveTangent=0.0797, interpMode=RCIM_Cubic, leaveTangent=0.0797, tangentMode=RCTM_User, time=1, value=1}
        ]}}})
    Module.Param0WriteEnabled=true
    Module.Param0WriteEnabled2=true
    Module.Param1WriteEnabled=true
    Module.Param1WriteEnabled2=true
    Module.Param2WriteEnabled=true
    Module.Param2WriteEnabled2=true
    Module.Param3WriteEnabled=true
    Module.Param3WriteEnabled2=true
    Write Parameter Index 2=true
   )
   &ScaleColor(
    Scale Alpha=&FloatFromCurve(FloatCurve=DataInterface() {propertyValues {Curve {keys [<at 0 0 RCIM_Cubic>,<at 0.2 0.975 RCIM_Cubic>,<at 1 0 RCIM_Cubic>]}}})
    ScaleA=true
    ScaleRGB=true
   )
   &ScaleSpriteSize(Uniform Curve Sprite Scale=DataInterface() {propertyValues {Curve {keys [
        {arriveTangent=1.0717, interpMode=RCIM_Cubic, leaveTangent=1.0717, tangentMode=RCTM_User, time=0, value=1}
        {arriveTangent=0.0885, interpMode=RCIM_Cubic, leaveTangent=0.0885, tangentMode=RCTM_User, time=1, value=1.7}
       ]}}})
  ]
  rendererClasses [&NiagaraSpriteRendererProperties]
  renderers [
   {rendererClass=&NiagaraSpriteRendererProperties}
   {rendererClass=&NiagaraSpriteRendererProperties, rendererIndex=1}
  ]
 }}
summary {
 systemName=NS_Emote_FirstClass_LandDust
 emitters [{emitterName=Dust, rendererClasses [&NiagaraSpriteRendererProperties]}]
 userVariables [
  User.FootDustColorArray {
   type=NiagaraDataInterfaceArrayColor
   description=""
   defaultValue {
    struct=NiagaraExt_VariableValue_DataInterface
    value {
     dataInterface=&self:NiagaraDataInterfaceArrayColor_0
     dataInterfaceClass=&NiagaraDataInterfaceArrayColor
    }
   }
  }
  User.SurfaceTypeIndex {type=NiagaraInt32}
 ]
}
